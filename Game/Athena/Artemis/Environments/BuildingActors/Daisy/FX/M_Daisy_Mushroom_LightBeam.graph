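# Unlit additive Niagara material for a light beam effect on the Daisy mushroom building actor, usable on sprite and mesh particles.
# Panning noise and light-ray textures form streaks, shaped along V by Beam Length and added to a bottom gradient.
# Particle color, a U-axis edge falloff and Fresnel shape this into the emissive term.
# A separate alpha term combines particle alpha, depth fade and an exponential radial gradient.
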Material /Game/Athena/Artemis/Environments/BuildingActors/Daisy/FX/M_Daisy_Mushroom_LightBeam
graph {
 properties {
  bUsedWithNiagaraMeshParticles=true
  bUsedWithNiagaraSprites=true
  blendMode=BLEND_Additive
  pixelDepthOffsetMode=PDOM_Legacy
  shadingModel=MSM_Unlit
  usedShadingModels=Unlit
 }
 nodes [
  n1 = MaterialFunctionCall(
   CenterPosition=AppendVector(A=ScalarParameter() {defaultValue=0.5, parameterName=Position_X}, B=ScalarParameter() {defaultValue=0.5, parameterName=Position_Y})
   Radius=ScalarParameter() {defaultValue=0.5, parameterName=Radius}
   Density=ScalarParameter() {defaultValue=2, parameterName=Density}
  ) {materialFunction=&RadialGradientExponential}
  n2 = ComponentMask(None=TextureCoordinate()) {g=true}
  Add(A=@n1.RadialGradientExponential, B=Power(Base=Multiply(A=@n2)) {constExponent=4})
  Constant2Vector() {g=0.5, r=0.5}
  n3 = TextureSample(UVs=Panner(Coordinate=TextureCoordinate() {uTiling=2, vTiling=0.1}) {bFractionalPart=true, speedX=0.0924, speedY=0.011}) {texture=&T_NoisePack_01}
  n4 = TextureSample(UVs=Panner(Coordinate=TextureCoordinate()) {bFractionalPart=true, speedX=-0.0679}) {texture=/Game/Effects/Fort_Effects/Textures/Outlander/T_InTheZone_LightRays_01}
  n5 = TextureCoordinate()
  n6 = Multiply(
   A=Multiply(A=Multiply(A=@n3.R, B=@n4.G), B=ScalarParameter() {defaultValue=4, parameterName=Streak Brightness})
   B=Power(Base=ComponentMask(None=@n5) {g=true}, Exp=ScalarParameter() {defaultValue=6, parameterName=Beam Length})
  )
  n7 = ComponentMask(None=TextureCoordinate()) {g=true}
  n8 = ScalarParameter() {defaultValue=6, parameterName=Bottom Gradiant}
  n9 = Multiply(A=ParticleColor().RGB, B=Add(A=Saturate(None=@n6), B=Power(Base=@n7, Exp=@n8)))
  n10 = TextureCoordinate()
  n11 = Multiply(A=ComponentMask(None=@n10) {r=true}, B=OneMinus(None=ComponentMask(None=@n10) {r=true}))
  n12 = Multiply(A=Power(Base=Saturate(None=@n11)) {constExponent=3}) {constB=11}
  n13 = ScalarParameter() {defaultValue=1, parameterName=Fresnel Brightness}
  n14 = ScalarParameter() {defaultValue=2, parameterName=Fresnel Exponent}
  MaterialFunctionCall(Emissive=Multiply(
    A=Multiply(A=@n9, B=@n12)
    B=MaterialFunctionCall(Brightness=@n13, Exponent=@n14) {materialFunction=&MF_Fresnel_FX}.Fresnel
   )) {materialFunction=&MF_FX_EmissiveAdjustments}
  n15 = Constant() {r=1}
  n16 = ScalarParameter() {defaultValue=14.9388, parameterName=DepthFadeDistance}
  Multiply(A=ParticleColor().A, B=Multiply(A=DepthFade(Opacity=@n15, FadeDistance=@n16), B=@n1.RadialGradientExponential))
 ]
}
