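# Niagara system for the idle glow and light beam on a Daisy mushroom building actor, using two looping sprite emitters.
# MaterialCard staggers two bursts half a lifetime apart so vertically aligned cyan cards cross-fade continuously, with User.BounceFlare stretching their height up to double.
# Lights adds a camera-offset blue sprite that pulses in size.
# User parameters drive size and offset, and time-of-day scalars adjust both emitters' brightness.
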NiagaraSystem /Game/Athena/Artemis/Environments/BuildingActors/Daisy/FX/NS_Daisy_Mushroom_LightBeam
dependencies {
 usedDataInterfaces [&NiagaraDataInterfaceCurve, &NiagaraDataInterfaceMeshRendererInfo]
 usedDynamicInputs []
 usedModules [&SystemState]
}
emitters {
 Lights {
  emitterUpdateScript [
   &EmitterState()
   &SpawnBurst_Instantaneous(Use Spawn Probability=false, UseLoopCountLimit=true)
  ]
  particleSpawnScript [
   &InitializeParticle(
    Color Mode=::NewEnumerator0
    Lifetime=2.8
    Position Offset=$User.IdlePositionOffset_Light
    Sprite Size=$User.IdleScale_Light
    Sprite Size Mode=::NewEnumerator4
    UsePositionOffset=true
    Write Material Random=true
   )
   &CameraOffset(Camera Offset Amount=22)
  ]
  particleUpdateScript [
   &ParticleState(
    Kill Particles When Lifetime Has Elapsed=false
    Let Infinitely Lived Particles Die When Emitter Deactivates=true
    Loop Particles Lifetime=true
   )
   &Color(
    Color=<rgba 0 0.5 1 1>
    Scale Color=&VectorFromFloat(Value=&FNTimeOfDayScalar(Night Scale=1.7, Morning Scale=1, Day Scale=2.3, Evening Scale=1))
   )
   &ScaleSpriteSize(
    Uniform Curve Scale=0.9
    Uniform Curve Sprite Scale=DataInterface() {propertyValues {Curve {keys [
        {interpMode=RCIM_Cubic, tangentMode=RCTM_User, time=0, value=1}
        {interpMode=RCIM_Cubic, tangentMode=RCTM_User, time=0.5, value=1.3}
        {tangentMode=RCTM_User, time=1, value=1}
       ]}}}
   ) {module=ScaleSpriteSize002}
  ]
  rendererClasses [&NiagaraSpriteRendererProperties]
  renderers [{rendererClass=&NiagaraSpriteRendererProperties}]
 }
 MaterialCard {
  emitterUpdateScript [
   &EmitterState()
   &SpawnBurst_Instantaneous(Use Spawn Probability=false, UseLoopCountLimit=true) {module=SpawnBurst_Instantaneous001}
   &SpawnBurst_Instantaneous(Spawn Time=2.5, Use Spawn Probability=false, UseLoopCountLimit=true)
  ]
  particleSpawnScript [
   &InitializeParticle(
    Color Mode=::NewEnumerator0
    Lifetime=5
    Position Offset=$User.IdlePositionOffset
    Sprite Size=&Multiply_Vector2D(A=$User.Scale, B=&Vector2DFromFloat(Value=$User.IdleScale))
    Sprite Size Mode=::NewEnumerator4
    UsePositionOffset=true
    Write Material Random=true
   )
   &SpriteFacingAndAlignment(Alignment Coordinate Space=::Local, Sprite Alignment=<0 0 1>, Sprite Facing=<0 0 1>, WriteAlignment=true, WriteFacing=false)
  ]
  particleUpdateScript [
   &ParticleState(
    Kill Particles When Lifetime Has Elapsed=false
    Let Infinitely Lived Particles Die When Emitter Deactivates=true
    Loop Particles Lifetime=true
   )
   &Color(
    Color=<rgba 0.568 2.7121 4 1>
    Scale Alpha=&FloatFromCurve(FloatCurve=DataInterface() {propertyValues {Curve {keys [
         {interpMode=RCIM_Cubic, tangentMode=RCTM_User, time=0, value=0}
         <at 0.3 1>
         {interpMode=RCIM_Cubic, tangentMode=RCTM_User, time=0.7, value=1}
         {tangentMode=RCTM_User, time=1, value=0}
        ]}}})
    Scale Color=&VectorFromFloat(Value=&FNTimeOfDayScalar(Night Scale=2, Morning Scale=1.5, Day Scale=3, Evening Scale=2))
   )
   &ScaleSpriteSize(
    Scale Factor=&MakeVector2D(X=1, Y=&Lerp_Float(A=1, B=2, Alpha=$User.BounceFlare))
    Scale Sprite Size Mode=::NewEnumerator2
   )
  ]
  rendererClasses [&NiagaraSpriteRendererProperties]
  renderers [{rendererClass=&NiagaraSpriteRendererProperties}]
 }
}
summary {
 systemName=NS_Daisy_Mushroom_LightBeam
 emitters [
  {emitterName=MaterialCard, rendererClasses [&NiagaraSpriteRendererProperties]}
  {emitterName=Lights, rendererClasses [&NiagaraSpriteRendererProperties]}
 ]
 userVariables [
  User.BounceFlare {type=NiagaraFloat, defaultValue=0, description=""}
  User.FXGameplayLifetime {type=NiagaraFloat, defaultValue=1, description=""}
  User.FXGameplayScale {type=NiagaraFloat, defaultValue=0, description=""}
  User.IdlePositionOffset {type=Vector3f, description="", defaultValue {struct=Vector3f, value=<0 0 0>}}
  User.IdlePositionOffset_Light {type=Vector3f, description="", defaultValue {struct=Vector3f, value=<0 0 0>}}
  User.IdleScale {type=NiagaraFloat, defaultValue=1, description=""}
  User.IdleScale_Light {type=Vector2f, description="", defaultValue {struct=Vector2f, value=<1 1>}}
  User.Scale {type=Vector2f, description="", defaultValue {struct=Vector2f, value=<99 55>}}
 ]
}
